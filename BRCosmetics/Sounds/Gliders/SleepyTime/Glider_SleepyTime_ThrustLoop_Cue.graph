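# Sound cue for the looping thrust audio of the SleepyTime glider cosmetic.
# A random node picks between two equally weighted, looping wave players, one for each loop variant.
# It plays at reduced volume through the glider thrust attenuation, the glider sound class and a cosmetic glider loop audio properties sheet.
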SoundCue /BRCosmetics/Sounds/Gliders/SleepyTime/Glider_SleepyTime_ThrustLoop_Cue
root=Random_1
nodes [
 {id=Random_1, children [WavePlayer_0,WavePlayer_3], properties {weights [1,1]}}
 {id=WavePlayer_0, properties {bLooping=true, soundWaveAssetPtr=&/Glider_SleepyTime_Loop_01}}
 {id=WavePlayer_3, properties {bLooping=true, soundWaveAssetPtr=&/Glider_SleepyTime_Loop_02}}
]
properties {
 attenuationSettings=&Fort_Player_Glider_Thrust_Athena_Att
 audioPropertiesSheet=&APS_Cosmetic_Glider_Loop
 maxDistance=12500
 soundClassObject=&Fort_Gliders_Land
 volumeMultiplier=0.2
}
